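# Physics asset for the hair bun part of the Nigiri Ninja medium female player character's body.
# Three capsule bodies cover the head, a tiny bun base joint and a larger dynamic bun joint.
# The base is angularly locked to the head, while the dynamic bun is constrained to the base with small swing and twist limit angles.
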PhysicsAsset /BRCosmetics/Characters/Player/Female/Medium/Bodies/F_MED_NigiriNinja/Meshes/Parts/F_MED_NigiriNinja_Bunn_Physics
bodies {
 base_bunHair_a_00_c_jnt [{shapeType=Capsule, length=0.505, radius=0.505, shapeName=base_bunHair_a_00_c_jnt_capsule, rotation=<rot 0 -90 0>}]
 dyb_bunHair_a_00_c_jnt [{shapeType=Capsule, length=0, radius=5, center=<3.1406 0 0>, rotation=<rot -90 180 180>}]
 head [{shapeType=Capsule, length=9, radius=7.4483, center=<1.3074 1.82 1.0508>, rotation=<rot 60 -90.0001 -70>}]
}
constraints [
 {bone1Name=dyb_bunHair_a_00_c_jnt, bone2Name=base_bunHair_a_00_c_jnt, swing1LimitDegrees=10, swing2LimitDegrees=10, twistLimitDegrees=10}
 {
  bone1Name=base_bunHair_a_00_c_jnt
  bone2Name=head
  swing1LimitDegrees=45
  swing1Motion=Locked
  swing2LimitDegrees=45
  swing2Motion=Locked
  twistLimitDegrees=45
  twistMotion=Locked
 }
]
