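# Physics asset for the Nigiri Ninja (female, medium) face accessory mesh, covering a right-side hair strand and two ribbons.
# Capsule bodies cover the neck, head, hair bones and both ribbons, with a second, thin angled capsule on the head.
# The hair base is rotationally locked to the head, as is the head to the neck.
# The two-segment hair chain and both ribbons have asymmetric swing limits and zero-degree twist limits for constrained secondary motion.
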PhysicsAsset /BRCosmetics/Characters/Player/Female/Medium/Bodies/F_MED_NigiriNinja/Meshes/Parts/F_MED_NigiriNinja_FaceAcc_Physics
bodies {
 base_hair_a_00_r_jnt [{shapeType=Capsule, length=0.505, radius=0.505, shapeName=base_hair_a_00_r_jnt_capsule, rotation=<rot 0 -90 0>}]
 base_ribbon_00_l [{shapeType=Capsule, length=1, radius=1, shapeName=base_ribbon_00_l_capsule, center=<2 0 0>, rotation=<rot 0 -60 -60>}]
 base_ribbon_00_r [{shapeType=Capsule, length=1, radius=1, shapeName=base_ribbon_00_r_capsule, center=<2 0 0>, rotation=<rot 0 -90 -70>}]
 dyn_hair_a_00_r_jnt [{shapeType=Capsule, length=4, radius=1.5, shapeName=dyn_hair_a_00_r_jnt_capsule, center=<2 0 0>, rotation=<rot 0 -90 -100>}]
 dyn_hair_a_01_r_jnt [{
   shapeType=Capsule
   length=5.813
   radius=0.5
   shapeName=dyn_hair_a_01_r_jnt_capsule
   center=<2.9384 -0.0518 0.0229>
   rotation=<rot -0.867 90.0754 -89.9631>
  }]
 head [
  {shapeType=Capsule, length=8, radius=6.5, center=<3.5421 1.3785 0>, rotation=<rot 90 0 15.0007>}
  {shapeType=Capsule, length=6.8, radius=0.6, shapeName=head_capsule, center=<-2.25 6.9839 6.3404>, rotation=<rot -60.6439 8.6087 12.1483>}
 ]
 neck_02 [{shapeType=Capsule, length=4, radius=3.505, shapeName=neck_02_capsule, rotation=<rot 0 -90.0001 90>}]
}
constraints [
 {bone1Name=dyn_hair_a_00_r_jnt, bone2Name=base_hair_a_00_r_jnt, swing1LimitDegrees=5, swing2LimitDegrees=45, twistLimitDegrees=0}
 {bone1Name=dyn_hair_a_01_r_jnt, bone2Name=dyn_hair_a_00_r_jnt, swing1LimitDegrees=5, swing2LimitDegrees=15, twistLimitDegrees=0}
 {
  bone1Name=base_hair_a_00_r_jnt
  bone2Name=head
  swing1LimitDegrees=45
  swing1Motion=Locked
  swing2LimitDegrees=45
  swing2Motion=Locked
  twistLimitDegrees=45
  twistMotion=Locked
 }
 {bone1Name=base_ribbon_00_l, bone2Name=head, swing1LimitDegrees=10, swing2LimitDegrees=20, twistLimitDegrees=0}
 {bone1Name=base_ribbon_00_r, bone2Name=head, swing1LimitDegrees=10, swing2LimitDegrees=30, twistLimitDegrees=0}
 {
  bone1Name=head
  bone2Name=neck_02
  swing1LimitDegrees=45
  swing1Motion=Locked
  swing2LimitDegrees=45
  swing2Motion=Locked
  twistLimitDegrees=45
  twistMotion=Locked
 }
]
