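# Post-process style secondary-motion AnimBlueprint for the NigiriNinja face accessory, following the parent mesh's copied pose and curves.
# Two neck_01-relative RigidBody sims run in series (face accessory, then bun at partial alpha).
# Weighted head/neck constraints then drive both ribbon bases, before zero-gravity AnimDynamics swings two chopsticks within angular limits.
# EventGraph updates Gravity Override through a deprecated function, but the AnimGraph uses fixed gravity values.
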AnimBlueprint /BRCosmetics/Characters/Player/Female/Medium/Bodies/F_MED_NigiriNinja/Meshes/Parts/F_MED_NigiriNinja_FaceAcc_AnimBP
target_skeleton=&/F_MED_NigiriNinja_FaceAcc_Skeleton
graphs [
 AnimGraph {nodes [
   n1 = RigidBody!(
    ComponentPose=Local To Component(LocalPose=Copy Pose From Mesh() {bCopyCurves=true, bUseAttachedParent=true})
    Alpha
    OverrideWorldGravity:Vector="0,0,-980"
    ExternalForce
   ) -> (Pose) {
    bOverrideWorldGravity=true
    overridePhysicsAsset=&/F_MED_NigiriNinja_FaceAcc_Physics
    simulationSpace=BaseBoneSpace
    alphaBoolBlend {blendOption=QuadraticInOut}
    baseBoneRef {boneName=neck_01}
    componentAppliedLinearAccClamp=<3000 3000 3000>
    componentLinearAccScale=<0.3 0.3 0.3>
    componentLinearVelScale=<0.3 0.3 0.3>
   }
   n2 = RigidBody(ComponentPose=@n1, Alpha=0.3, OverrideWorldGravity="0,0,-980") {
    bOverrideWorldGravity=true
    overridePhysicsAsset=&/F_MED_NigiriNinja_Bunn_Physics
    simulationSpace=BaseBoneSpace
    alphaBoolBlend {blendOption=Cubic}
    baseBoneRef {boneName=neck_01}
    componentAppliedLinearAccClamp=<3000 3000 3000>
    componentLinearAccScale=<0.3 0.3 0.3>
    componentLinearVelScale=<0.3 0.3 0.3>
   }
   n3 = "Constraint - base_ribbon_00_r (PPP)"!(ComponentPose=@n2, Alpha=1, ConstraintWeights_0=0.4048, ConstraintWeights_1:float=0.3508, ConstraintWeights_2:float=0.2444) -> (Pose) {
    alphaBoolBlend {blendOption=Linear}
    boneToModify {boneName=base_ribbon_00_r}
    constraintSetup [{targetBone {boneName=head}}, {targetBone {boneName=neck_01}}, {targetBone {boneName=neck_02}}]
    constraintWeights [1,1,1]
   }
   n4 = "Constraint - base_ribbon_00_l (PPP)"!(ComponentPose=@n3, Alpha=1, ConstraintWeights_0=0.4048, ConstraintWeights_1:float=0.3508, ConstraintWeights_2:float=0.2444) -> (Pose) {
    alphaBoolBlend {blendOption=Linear}
    boneToModify {boneName=base_ribbon_00_l}
    constraintSetup [{targetBone {boneName=head}}, {targetBone {boneName=neck_01}}, {targetBone {boneName=neck_02}}]
    constraintWeights [1,1,1]
   }
   n5 = Anim Dynamics - Bone: dyn_chopstick_a_00_jnt!(ComponentPose=@n4, Alpha=1, GravityOverride:Vector="0,0,0", ExternalForce="0,0,0") -> (Pose) {
    bUseGravityOverride=true
    bUsePlanarLimit=false
    simulationSpace=BoneRelative
    alphaBoolBlend {blendOption=Linear}
    boundBone {boneName=dyn_chopstick_a_00_jnt}
    componentAppliedLinearAccClamp=<1000 1000 1000>
    componentLinearAccScale=<0.25 0.25 0.25>
    componentLinearVelScale=<0.25 0.24 0.25>
    physicsBodyDefinitions [{
      boundBone {boneName=dyn_chopstick_a_00_jnt}
      boxExtents=<10 8 8>
      constraintSetup {angularLimitsMax=<0 0.5 0.5>, angularLimitsMin=<0 -0.5 -0.3>, angularTarget=<0 0 0>}
      localJointOffset=<1 0 0>
     }]
    relativeSpaceBone {boneName=root}
    sphericalLimits [{limitRadius=0, drivingBone {boneName=head}, sphereLocalOffset=<0 0 0>}]
   }
   n6 = Anim Dynamics - Bone: dyn_chopstick_b_00_jnt!(ComponentPose=@n5, Alpha=1, GravityOverride:Vector="0,0,0", ExternalForce="0,0,0") -> (Pose) {
    bUseGravityOverride=true
    simulationSpace=BoneRelative
    alphaBoolBlend {blendOption=Linear}
    boundBone {boneName=dyn_chopstick_b_00_jnt}
    componentAppliedLinearAccClamp=<1000 1000 1000>
    componentLinearAccScale=<0.25 0.25 0.25>
    componentLinearVelScale=<0.25 0.25 0.25>
    physicsBodyDefinitions [{
      boundBone {boneName=dyn_chopstick_b_00_jnt}
      boxExtents=<10 8 8>
      constraintSetup {angularLimitsMax=<0 0.5 0.3>, angularLimitsMin=<0 -0.5 -0.5>, angularTarget=<0 0 0>}
      localJointOffset=<1 0 0>
     }]
    relativeSpaceBone {boneName=root}
   }
   Output Pose(Result=Component To Local(ComponentPose=@n6)) {layerGroup=None}
  ]}
 EventGraph {nodes [
   Event BlueprintUpdateAnimation()
   > n1 = Gravity Override State Function DEPRECATED(CharacterPartAnimInstance=Self-Reference(), GameTimeInSeconds=GetGameTimeInSeconds(), GravityOverrideParams=Get Gravity Override Params())
   > Set Gravity Override(Gravity Override=@n1)
  ]}
 OnUpdate {nodes [
   OnUpdate!() -> (then, Context:"Anim Update Context Structure (by ref)", Node:"Anim Node Reference Structure (by ref)") {is=title K2Node_FunctionEntry}
   > Return Node()
  ]}
]
variables [
 Body_Mesh {type=SkeletalMeshComponent}
 BunAlpha {type=number, default=1}
 Gravity Override {type=Vector}
 Gravity Override Params {type=GravityOverrideParamsStruct}
 NewVar {type=SceneComponent, default=null}
]
